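# Niagara effect for the HornetBerm emote that places a dirt-colored ground crack under the main player.
# A single GroundCrack emitter bursts a long-lived particle, offset slightly downward in local space, after a short delay, with both sprite and decal renderers.
# Decal curves quickly grow its footprint and fade it in, then hold.
# A front-end float modifier makes the decal fade weaker in gameplay than in the front end.
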NiagaraSystem /BRCosmetics/Animation/Game/MainPlayer/Emotes/HornetBerm/FX/NS_Emote_HornetBerm_Dirt
dependencies {
 usedDataInterfaces [&NiagaraDataInterfaceVectorCurve, &NiagaraDataInterfaceCurve]
 usedDynamicInputs []
 usedModules [&SystemState]
}
emitters {GroundCrack {
  emitterUpdateScript [
   &EmitterState(Life Cycle Mode=::NewEnumerator1, Loop Behavior=::NewEnumerator1, Loop Delay=0.1, Loop Duration=21, UseLoopDelay=true)
   &SpawnBurst_Instantaneous(Use Spawn Probability=false, UseLoopCountLimit=false)
  ]
  particleSpawnScript [
   &InitializeParticle(
    Color=<rgba 0.0133 0.0074 0.003 1>
    Lifetime=21
    Position Offset=<0 0 -5>
    Position Offset Coordinate Space=::Local
    Sprite Size Mode=::NewEnumerator3
    Uniform Sprite Size=200
    UsePositionOffset=true
    Write Material Random=true
   )
   &SpriteFacingAndAlignment(Sprite Facing=<0 0 1>, WriteAlignment=false, WriteFacing=true)
  ]
  particleUpdateScript [
   &ParticleState()
   /Niagara/Modules/Decal/Decal_Attributes(
    Decal Fade:NiagaraFloat=&FloatFromCurve(
     FloatCurve=DataInterface() {propertyValues {Curve {keys [<at 0 0>,<at 0.005 1>,<at 0.6494 1>]}}}
     Scale Curve=&FN_FrontEndFloatModifier(Gameplay=0.4, Front End=1)
    )
    Decal Size:Vector3f=&VectorFromCurve(
     VectorCurve=DataInterface() {propertyValues {XCurve {keys [<at 0 1>,<at 1 1>]}, YCurve {keys [<at 0 0.5>,<at 0.005 1>]}, ZCurve {keys [<at 0 0.5>,<at 0.005 1>]}}}
     Scale Curve=<40 150 150>
    )
    Fade=1
    Size=1
   )
  ]
  rendererClasses [&NiagaraSpriteRendererProperties, &NiagaraDecalRendererProperties]
  renderers [
   {rendererClass=&NiagaraSpriteRendererProperties}
   {rendererClass=&NiagaraDecalRendererProperties, rendererIndex=1}
  ]
 }}
summary {
 systemName=NS_Emote_HornetBerm_Dirt
 emitters [{emitterName=GroundCrack, rendererClasses [&NiagaraSpriteRendererProperties, &NiagaraDecalRendererProperties]}]
 userVariables [
  User.bIsFrontEnd {type=NiagaraBool}
  User.bIsFrontEndPreview {type=NiagaraBool}
 ]
}
