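# Niagara effect for the HornetBerm emote's prop spawn, throwing debris, dust and rocks.
# Debris (meshes and sprites split by spawn group) and two dust emitters fire immediately, while Rocks_Burst and Dust002 share a delayed second burst about 1.4 seconds later.
# Debris picks collision by quality level, including a local-plane variant, and some dust opacity follows time of day.
# GroundCrack decal emitter and all rock collision modules are disabled.
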NiagaraSystem /BRCosmetics/Animation/Game/MainPlayer/Emotes/HornetBerm/FX/NS_Emote_HornetBerm_PropSpawn
dependencies {
 usedDataInterfaces [
  &NiagaraDataInterfacePlatformSet
  &NiagaraDataInterfaceCurve
  &NiagaraDataInterfaceVectorCurve
  &NiagaraDataInterfaceDebugDraw
  &NiagaraDataInterfaceColorCurve
 ]
 usedDynamicInputs [
  &RandomRangeVector2D
  &MultiplyLinearColor
  &MakeVectorFromLinearColorRGB
  &FNTimeOfDayColor
  &MakeVector4_FromLinearColor
  &ColorFromCurve
 ]
 usedModules [
  &SystemState
  &self:SystemSpawnScript.SystemScriptSource.SystemScriptGraph.NiagaraNodeAssignment_0.SetVariables_#1
  &self:SystemSpawnScript.SystemScriptSource.SystemScriptGraph.NiagaraNodeAssignment_1.SetVariables_#2
 ]
}
emitters {
 Debris {
  emitterUpdateScript [
   &EmitterState(Life Cycle Mode=::NewEnumerator1, Loop Behavior=::NewEnumerator1, Loop Duration=0.5, UseLoopDelay=false)
   &SpawnBurst_Instantaneous(Spawn Count=4, Use Spawn Probability=false, UseLoopCountLimit=false)
   &SpawnRate(Spawn Probability=0.8, SpawnRate=8, Use Spawn Probability=true)
   &SpawnRate(Spawn Probability=0.8, SpawnGroup=1, SpawnRate=6, Use Spawn Probability=true) {module=SpawnRate001}
  ]
  particleSpawnScript [
   &FN_SetSpawnGroup(Number of Spawn Groups=2)
   /SetVariables(Particles.VisibilityTag=$Particles.SpawnGroup) {isSetParameters=true}
   &InitializeParticle(
    Color Maximum=&MergeHelper_ColorSwitch(0=<rgba 0.4921 0.5944 0.7083 1>, 1=<rgba 0.375 0.2328 0.1562 0.5>, 2=?, 3=?, Index=$Particles.VisibilityTag, Num=::NewEnumerator0)
    Color Minimum=&MergeHelper_ColorSwitch(0=<rgba 0.5925 0.8193 0.9844 1>, 1=<rgba 0.375 0.2073 0.1562 0.35>, 2=?, 3=?, Index=$Particles.VisibilityTag, Num=::NewEnumerator0)
    Color Mode=::NewEnumerator2
    Lifetime Max=0.5
    Lifetime Mode=::NewEnumerator1
    Mass Mode=::NewEnumerator1
    Mesh Scale Max=<1 1 1>
    Mesh Scale Min=<0.5 0.5 0.5>
    Mesh Scale Mode=::NewEnumerator2
    Position Offset Coordinate Space=::Local
    Sprite Rotation Mode=::NewEnumerator2
    Sprite Size Mode=::NewEnumerator1
    Sprite UV Mode=::NewEnumerator3
    Uniform Sprite Size Max=15
    Uniform Sprite Size Min=3
    UsePositionOffset=true
    Write Material Random=true
   )
   &ShapeLocation(MaskSpawnGroup=false, Offset=<0 0 30>, Offset Mode=::NewEnumerator0, Override Seed=false, Sphere Radius=30, UseOwnerScale=true)
   &AddVelocity(
    Module.Use Falloff Distance=false
    Override Seed=false
    Velocity Mode=::NewEnumerator1
    Velocity Speed=&RandomRangeFloat(Minimum=100, Maximum=200)
   )
   &AddVelocity(Velocity=&RandomRangeVector(Minimum=<0 0 100>, Maximum=<0 0 200>)) {module=AddVelocity001}
   /SetVariables(Particles.MeshIndex=&RandomRangeInt(Minimum=0, Maximum=2)) {isSetParameters=true}
   &InitialMeshOrientation(Use Rotation Vector=false)
  ]
  particleUpdateScript [
   &ParticleState()
   &UpdateMeshOrientation(
    Rotation Rate=&RandomRangeFloat(Minimum=1, Maximum=4)
    Rotation Vector=&RandomRangeVector(Minimum=<-1 -1 -1>, Maximum=<1 1 1>)
   )
   &GravityForce(Gravity=&RandomRangeVector(Minimum=<0 0 -680>, Maximum=<1 1 -980>))
   &Drag(Use Linear Drag=true, Use Rotational Drag=true)
   &Collision(
    Age Colliding Particles=false
    Collision Enabled=&Boolean_AndOperation(
     Bool1=&IsPlatformSetActive(New Platform Set=DataInterface() {propertyValues {Platforms {qualityLevelMask=-8}}})
     Bool2=&SetBoolByFloatComparison(
      A=&FloatFromCurve(FloatCurve=DataInterface() {propertyValues {Curve {
          defaultValue=1
          keys [{interpMode=RCIM_Cubic, tangentMode=RCTM_User, time=0.0045, value=0}, {interpMode=RCIM_Cubic, tangentMode=RCTM_User, time=0.1114, value=1}]
         }}})
      B=0.5
     )
    )
    KillOnCollision=false
    Module.Control Rest State Via Surface Normal=false
    Module.EnableMaxCollisionCount=false
    Module.Randomize Collision Normal Vector=false
    Radius Calculation Type=::NewEnumerator1
   )
   &Collision(
    Age Colliding Particles=false
    Analytical Collision Plane Position 1:Vector3f=<0 0 -20>
    Analytical Collision Plane Space:ENiagaraCoordinateSpace=::Local
    CPU Collision Type=::NewEnumerator3
    Collision Enabled=&Boolean_AndOperation(
     Bool1=&IsPlatformSetActive(New Platform Set=DataInterface() {propertyValues {Platforms {qualityLevelMask=-28}}})
     Bool2=&SetBoolByFloatComparison(
      A=&FloatFromCurve(FloatCurve=DataInterface() {propertyValues {Curve {
          defaultValue=1
          keys [{interpMode=RCIM_Cubic, tangentMode=RCTM_User, time=0.0045, value=0}, {interpMode=RCIM_Cubic, tangentMode=RCTM_User, time=0.1114, value=1}]
         }}})
      B=0.5
     )
    )
    KillOnCollision=false
    Module.Control Rest State Via Surface Normal=false
    Module.EnableMaxCollisionCount=false
    Module.Randomize Collision Normal Vector=false
    Perform 2 Analytical Collision Queries=false
   ) {module=Collision001}
   &SolveForcesAndVelocity(Clamp Velocity=false, Limit Acceleration=false)
   &ScaleSpriteSize(Scale Factor=&Vector2DFromFloat(Value=&FloatFromCurve(FloatCurve=DataInterface() {propertyValues {Curve {keys [<at 0.5 1 RCIM_Cubic>, {arriveTangent=-3.7562, interpMode=RCIM_Cubic, leaveTangent=-3.7562, tangentMode=RCTM_User, time=1, value=0}]}}})))
   &ScaleMeshSize(Scale Factor=&VectorFromFloat(Value=&FloatFromCurve(FloatCurve=DataInterface() {propertyValues {Curve {keys [
          {arriveTangent=-0.008, interpMode=RCIM_Cubic, leaveTangent=-0.008, tangentMode=RCTM_User, time=0, value=1}
          {arriveTangent=-0.1006, interpMode=RCIM_Cubic, leaveTangent=-0.1006, tangentMode=RCTM_User, time=0.9, value=0}
         ]}}})))
  ]
  rendererClasses [&NiagaraMeshRendererProperties, &NiagaraSpriteRendererProperties]
  renderers [
   {rendererClass=&NiagaraMeshRendererProperties, rendererIndex=0}
   {rendererClass=&NiagaraSpriteRendererProperties, rendererIndex=1}
  ]
 }
 Dust {
  emitterUpdateScript [
   &EmitterState(Inactive Response=::NewEnumerator2, Life Cycle Mode=::NewEnumerator1, Loop Behavior=::NewEnumerator1, Loop Duration=0.3, UseLoopDelay=false)
   &SpawnBurst_Instantaneous(Spawn Count=16, Use Spawn Probability=false, UseLoopCountLimit=false)
   &SpawnRate(SpawnGroup=1, SpawnRate=7, Use Spawn Probability=false) {enabled=false, module=SpawnRate001}
   &SpawnRate(SpawnRate=12, Use Spawn Probability=false)
  ]
  particleSpawnScript [
   &InitializeParticle(
    Color=&Multiply_LinearColorByFloat(Linear Color=$System.Global color, Float=1)
    Lifetime=&RandomRangeFloat(Minimum=0.5, Maximum=1)
    Sprite Rotation Angle=&RandomRangeFloat(Minimum=0, Maximum=360)
    Sprite Rotation Mode=::NewEnumerator0
    Sprite Size Mode=::NewEnumerator3
    Uniform Sprite Size=&RandomRangeFloat(Minimum=20, Maximum=10)
    UsePositionOffset=false
    Write Material Random=true
   )
   &ShapeLocation(
    MaskSpawnGroup=false
    Non Uniform Scale=<1 1 0.1>
    Offset=<0 0 5>
    Offset Mode=::NewEnumerator0
    Override Seed=false
    Sphere Radius=10
    Sphere Surface Distribution=0.6601
    UseOwnerScale=true
   )
   &AddVelocity(Velocity=<0 0 100>) {enabled=false}
   &AddVelocity(
    Module.Use Falloff Distance=false
    Override Seed=false
    Velocity Mode=::NewEnumerator1
    Velocity Speed=&RandomRangeFloat(Minimum=80, Maximum=150)
   ) {module=AddVelocity001}
   &CameraOffset(Camera Offset Amount=5)
  ]
  particleUpdateScript [
   &ParticleState()
   &ScaleSpriteSize(
    Uniform Curve Scale=5
    Uniform Curve Sprite Scale=DataInterface() {propertyValues {Curve {defaultValue=1, keys [{tangentMode=RCTM_User, time=0, value=0.7}, {tangentMode=RCTM_User, time=1, value=1}]}}}
   )
   &SubUVAnimation(
    End Frame=15
    SubUV Animation Mode:ENiagara_SubUVLookupMode=::NewEnumerator1
    SubUV Curve=DataInterface() {propertyValues {Curve {keys [{tangentMode=RCTM_User, time=0, value=0}, {tangentMode=RCTM_User, time=1.0138, value=0.388}]}}}
   )
   &Drag(
    Drag=&FloatFromCurve(
     FloatCurve=DataInterface() {propertyValues {Curve {
        defaultValue=1
        keys [{interpMode=RCIM_Cubic, tangentMode=RCTM_User, time=0, value=0}, {interpMode=RCIM_Cubic, tangentMode=RCTM_User, time=0.3793, value=1}]
       }}}
     Scale Curve=7
    )
    Use Linear Drag=true
    Use Rotational Drag=true
   )
   &SolveForcesAndVelocity(Clamp Velocity=false, Limit Acceleration=false)
   /SetVariables(Particles.1 - Ground Fog Scale Color:LinearColor=&MakeLinearColorFromVectorAndFloat(
     "Vector (RGB)"=<1 1 1>
     "Float (Alpha)"=&FloatFromCurve(
      FloatCurve=DataInterface() {propertyValues {Curve {keys [
          {arriveTangent=5, interpMode=RCIM_Cubic, time=0, value=0}
          {arriveTangent=1.5181, interpMode=RCIM_Cubic, leaveTangent=1.5181, time=0.2, value=0.7}
          {
           arriveTangent=0.1817
           arriveTangentWeight=0.1842
           interpMode=RCIM_Cubic
           leaveTangent=0.1817
           leaveTangentWeight=0.7369
           tangentMode=RCTM_User
           time=0.5037
           value=0.7647
          }
          <at 1 0>
         ]}}}
      Scale Curve=&RandomRangeFloat(
       Minimum=&FNTimeOfDayScalar(Night Scale=0.035, Morning Scale=0.07, Day Scale=0.05, Evening Scale=0.07)
       Maximum=&FNTimeOfDayScalar(Night Scale=0.05, Morning Scale=0.07, Day Scale=0.07, Evening Scale=0.1)
      )
     )
    )) {enabled=false, isSetParameters=true, module=SetVariables_#3}
   /SetVariables(Particles.1 - Ground Fog Dyn Param:Vector4f=&MakeVector4(X=0, Y=1, Z=0.1, W=1)) {isSetParameters=true, module=SetVariables_#4}
   &ScaleColor(
    Scale Alpha=&FloatFromCurve(
     FloatCurve=DataInterface() {propertyValues {Curve {keys [
         {arriveTangent=6.4271, interpMode=RCIM_Cubic, leaveTangent=6.4271, tangentMode=RCTM_User, time=0, value=0}
         {arriveTangent=0.5834, interpMode=RCIM_Cubic, leaveTangent=0.5834, tangentMode=RCTM_User, time=0.0598, value=0.6812}
         {
          arriveTangent=0.1817
          arriveTangentWeight=0.1842
          interpMode=RCIM_Cubic
          leaveTangent=0.1817
          leaveTangentWeight=0.7369
          tangentMode=RCTM_User
          time=0.5037
          value=0.8
         }
         <at 1 0>
        ]}}}
     Scale Curve=&RandomRangeFloat(
      Minimum=&FNTimeOfDayScalar(Night Scale=0.8, Morning Scale=0.8, Day Scale=0.8, Evening Scale=0.8)
      Maximum=&FNTimeOfDayScalar(Night Scale=1, Morning Scale=1, Evening Scale=1)
     )
    )
    ScaleA=true
    ScaleRGB=true
   )
  ]
  rendererClasses [&NiagaraSpriteRendererProperties]
  renderers [{rendererClass=&NiagaraSpriteRendererProperties}]
 }
 Dust001 {
  emitterUpdateScript [
   &EmitterState(Life Cycle Mode=::NewEnumerator1, Loop Behavior=::NewEnumerator1, Loop Duration=0.3, UseLoopDelay=false)
   &SpawnBurst_Instantaneous(Spawn Count=9, Use Spawn Probability=false, UseLoopCountLimit=false)
   &SpawnRate(
    SpawnRate=&Multiply_Float(
     A=5
     B=&FloatFromCurve(
      FloatCurve=DataInterface() {propertyValues {Curve {keys [<at 0.0809 1.7>,<at 0.2058 0.7749>,<at 0.336 1.3794>,<at 0.4985 0.7333>,<at 0.7144 1.4627>,<at 0.8629 1>]}}}
      CurveIndex=$Emitter.NormalizedLoopAge
      Scale Curve=2
     )
    )
    Use Spawn Probability=false
   )
  ]
  particleSpawnScript [
   &InitializeParticle(
    Color=&RandomRangeLinearColor(Minimum=<rgba 0.0938 0.0567 0.0396 1.2>, Maximum=<rgba 0.0365 0.0325 0.0314 1.2>)
    Lifetime Mode=::NewEnumerator1
    Sprite Rotation Mode=::NewEnumerator2
    Sprite Size Mode=::NewEnumerator1
    Uniform Sprite Size Max=50
    Uniform Sprite Size Min=150
    UsePositionOffset=false
    Write Material Random=true
   )
   &SubUVAnimation(End Frame=3, SubUV Animation Mode:ENiagara_SubUVLookupMode=::NewEnumerator2)
   &CameraOffset(Camera Offset Amount=5)
   &ShapeLocation(
    MaskSpawnGroup=false
    Non Uniform Scale=<1 2 0.5>
    Offset=<0 0 5>
    Offset Mode=::NewEnumerator0
    Override Seed=false
    Rotation Coordinate Space=::World
    Shape Origin=&MakeVector(X=0, Y=0, Z=0)
    Sphere Radius=10
    UseOwnerScale=true
   )
   &AddVelocityFromPoint(
    Module.Use Falloff Distance=false
    Origin Offset=<0 0 -70>
    Velocity Origin:Vector3f=&MakeVector(X=0, Y=0, Z=0)
    Velocity Strength=&Multiply_Float(
     A=&RandomRangeFloat(Minimum=100, Maximum=200)
     B=&FloatFromCurve(
      FloatCurve=DataInterface() {propertyValues {Curve {keys [<at 0.0809 1.4>,<at 0.2058 0.8>,<at 0.336 1.2>,<at 0.4985 0.8>,<at 0.7144 1.2>,<at 0.8629 1>]}}}
      CurveIndex=$Emitter.NormalizedLoopAge
      Scale Curve=1.2
     )
    )
   )
  ]
  particleUpdateScript [
   &ParticleState()
   &SpriteRotationRate(Rotation Rate=&RandomRangeFloat(Minimum=-100, Maximum=100))
   &GravityForce(Gravity=&RandomRangeVector(Minimum=<0 0 -980>, Maximum=<0 0 -383.0567>))
   &Drag(Rotational Drag=5, Use Linear Drag=true, Use Rotational Drag=true) {module=Drag001}
   &SolveForcesAndVelocity(Clamp Velocity=false, Limit Acceleration=false)
   &ScaleSpriteSize(Uniform Curve Sprite Scale=DataInterface() {propertyValues {Curve {keys [
        {arriveTangent=4.1785, interpMode=RCIM_Cubic, leaveTangent=4.1785, tangentMode=RCTM_User, time=0, value=0}
        {arriveTangent=0.9995, interpMode=RCIM_Cubic, leaveTangent=0.9995, tangentMode=RCTM_User, time=0.2223, value=0.6825}
        {arriveTangent=-0.0405, interpMode=RCIM_Cubic, leaveTangent=-0.0405, tangentMode=RCTM_User, time=1, value=1}
       ]}}})
   &ScaleColor(
    Scale Alpha=&FloatFromCurve(
     FloatCurve=DataInterface() {propertyValues {Curve {keys [
         {arriveTangent=-0.4559, interpMode=RCIM_Cubic, leaveTangent=-0.4559, tangentMode=RCTM_User, time=0, value=1}
         {arriveTangent=-2.4378, interpMode=RCIM_Cubic, leaveTangent=-2.4378, tangentMode=RCTM_User, time=1.0011, value=0.002}
        ]}}}
     Scale Curve=0.8
    )
    Scale RGB=&Multiply_VectorByFloat(Vector=<1 1 1>, Float=&RandomRangeFloat(Minimum=0.5, Maximum=1))
    ScaleA=true
    ScaleRGB=true
   )
   /SetVariables(Particles.Position2:NiagaraPosition=&AddVectorToPosition(Position=$Engine.Emitter.SimulationPosition, Vector=Linked() {linkedVariable {name=Particles.Position, type=NiagaraPosition}})) {isSetParameters=true}
  ]
  rendererClasses [&NiagaraSpriteRendererProperties]
  renderers [{rendererClass=&NiagaraSpriteRendererProperties}]
 }
 Dust002 {
  emitterUpdateScript [
   &EmitterState(Life Cycle Mode=::NewEnumerator1, Loop Behavior=::NewEnumerator1, Loop Delay=1.438, Loop Duration=1, UseLoopDelay=true)
   &SpawnBurst_Instantaneous(Spawn Count=5, Use Spawn Probability=false, UseLoopCountLimit=false)
   &SpawnRate(SpawnGroup=1, SpawnRate=7, Use Spawn Probability=false) {enabled=false, module=SpawnRate001}
   &SpawnRate(SpawnRate=7, Use Spawn Probability=false) {enabled=false}
  ]
  particleSpawnScript [
   &InitializeParticle(
    Color=&Multiply_LinearColorByFloat(
     Linear Color=&RandomRangeLinearColor(Minimum=<rgba 0.0828 0.0644 0.0518 1>, Maximum=<rgba 0.0503 0.0367 0.0212 1>)
     Float=1
    )
    Lifetime=&RandomRangeFloat(Minimum=0.7, Maximum=0.2)
    Position Offset=<0 -7 40>
    Sprite Rotation Angle=&RandomRangeFloat(Minimum=0, Maximum=360)
    Sprite Rotation Mode=::NewEnumerator0
    Sprite Size Mode=::NewEnumerator3
    Uniform Sprite Size=&RandomRangeFloat(Minimum=30, Maximum=60)
    UsePositionOffset=true
    Write Material Random=true
   )
   &ShapeLocation(MaskSpawnGroup=false, Non Uniform Scale=<1 1 0.1>, Offset Mode=::NewEnumerator0, Override Seed=false, Sphere Radius=25, UseOwnerScale=true) {enabled=false}
   &AddVelocity(Velocity=<-30 0 100>)
   &AddVelocity(
    Module.Use Falloff Distance=false
    Origin Offset=<0 -7 40>
    Override Seed=false
    Velocity Mode=::NewEnumerator1
    Velocity Speed=&RandomRangeFloat(Minimum=80, Maximum=30)
   ) {module=AddVelocity001}
  ]
  particleUpdateScript [
   &ParticleState()
   &ScaleSpriteSize(Uniform Curve Sprite Scale=DataInterface() {propertyValues {Curve {defaultValue=1, keys [{tangentMode=RCTM_User, time=0, value=0.7}, {tangentMode=RCTM_User, time=1, value=1}]}}})
   &GravityForce(Gravity=&RandomRangeVector(Minimum=<0 0 -165.1874>, Maximum=<0 0 -479.7808>)) {enabled=false}
   &SubUVAnimation(
    End Frame=15
    SubUV Animation Mode:ENiagara_SubUVLookupMode=::NewEnumerator1
    SubUV Curve=DataInterface() {propertyValues {Curve {keys [{tangentMode=RCTM_User, time=0, value=0}, {tangentMode=RCTM_User, time=1.0138, value=0.388}]}}}
   )
   &Drag(
    Drag=&FloatFromCurve(
     FloatCurve=DataInterface() {propertyValues {Curve {
        defaultValue=1
        keys [{interpMode=RCIM_Cubic, tangentMode=RCTM_User, time=0, value=0}, {interpMode=RCIM_Cubic, tangentMode=RCTM_User, time=0.3793, value=1}]
       }}}
     Scale Curve=7
    )
    Use Linear Drag=true
    Use Rotational Drag=true
   )
   &SolveForcesAndVelocity(Clamp Velocity=false, Limit Acceleration=false)
   /SetVariables(
    Particles.0 - Env Fog Scale Color:LinearColor=&MakeLinearColorFromVectorAndFloat(
     "Vector (RGB)"=<1 1 1>
     "Float (Alpha)"=&FloatFromCurve(
      FloatCurve=DataInterface() {propertyValues {Curve {keys [
          {arriveTangent=3.5962, interpMode=RCIM_Cubic, leaveTangent=3.5962, tangentMode=RCTM_User, time=0, value=0}
          {
           arriveTangent=0.1817
           arriveTangentWeight=0.1842
           interpMode=RCIM_Cubic
           leaveTangent=0.1817
           leaveTangentWeight=0.7369
           tangentMode=RCTM_User
           time=0.5037
           value=0.7647
          }
          {arriveTangent=-4.7233, tangentMode=RCTM_User, time=1, value=0}
         ]}}}
      Scale Curve=&RandomRangeFloat(
       Minimum=0.02
       Maximum=&FNTimeOfDayScalar(Night Scale=0.05, Morning Scale=0.05, Day Scale=0.07, Evening Scale=0.05)
      )
     )
    )
    Particles.1 - Ground Fog Scale Color:LinearColor=&MakeLinearColorFromVectorAndFloat(
     "Vector (RGB)"=<1 1 1>
     "Float (Alpha)"=&FloatFromCurve(
      FloatCurve=DataInterface() {propertyValues {Curve {keys [
          {arriveTangent=5, interpMode=RCIM_Cubic, time=0, value=0}
          {arriveTangent=1.5181, interpMode=RCIM_Cubic, leaveTangent=1.5181, time=0.2, value=0.7}
          {
           arriveTangent=0.1817
           arriveTangentWeight=0.1842
           interpMode=RCIM_Cubic
           leaveTangent=0.1817
           leaveTangentWeight=0.7369
           tangentMode=RCTM_User
           time=0.5037
           value=0.7647
          }
          <at 1 0>
         ]}}}
      Scale Curve=&RandomRangeFloat(
       Minimum=&FNTimeOfDayScalar(Night Scale=0.035, Morning Scale=0.07, Day Scale=0.05, Evening Scale=0.07)
       Maximum=&FNTimeOfDayScalar(Night Scale=0.05, Morning Scale=0.07, Day Scale=0.07, Evening Scale=0.1)
      )
     )
    )
   ) {isSetParameters=true, module=SetVariables_#3}
   /SetVariables(
    Particles.0 - Env Fog Dyn Param:Vector4f=&MakeVector4(X=0, Y=1, Z=1, W=1)
    Particles.1 - Ground Fog Dyn Param:Vector4f=&MakeVector4(X=0, Y=1, Z=0.1, W=1)
   ) {isSetParameters=true, module=SetVariables_#4}
   &ScaleColor(
    Scale Alpha=&FloatFromCurve(
     FloatCurve=DataInterface() {propertyValues {Curve {keys [
         {arriveTangent=6.4271, interpMode=RCIM_Cubic, leaveTangent=6.4271, tangentMode=RCTM_User, time=0, value=0}
         {arriveTangent=0.5834, interpMode=RCIM_Cubic, leaveTangent=0.5834, tangentMode=RCTM_User, time=0.0598, value=0.6812}
         {
          arriveTangent=0.1817
          arriveTangentWeight=0.1842
          interpMode=RCIM_Cubic
          leaveTangent=0.1817
          leaveTangentWeight=0.7369
          tangentMode=RCTM_User
          time=0.5037
          value=0.8
         }
         <at 1 0>
        ]}}}
     Scale Curve=&RandomRangeFloat(
      Minimum=&FNTimeOfDayScalar(Night Scale=0.1, Morning Scale=0.2, Day Scale=0.3, Evening Scale=0.2)
      Maximum=&FNTimeOfDayScalar(Night Scale=0.3, Morning Scale=0.4, Day Scale=0.5, Evening Scale=0.4)
     )
    )
    ScaleA=true
    ScaleRGB=true
   )
  ]
  rendererClasses [&NiagaraSpriteRendererProperties]
  renderers [{rendererClass=&NiagaraSpriteRendererProperties}]
 }
 GroundCrack {
  bEnabled=false
  emitterUpdateScript [
   &EmitterState(Life Cycle Mode=::NewEnumerator1, Loop Behavior=::NewEnumerator1, Loop Delay=0.1, Loop Duration=3, UseLoopDelay=true)
   &SpawnBurst_Instantaneous(Use Spawn Probability=false, UseLoopCountLimit=false)
  ]
  particleSpawnScript [
   &InitializeParticle(
    Color=<rgba 0.0133 0.0074 0.003 1>
    Lifetime=3
    Position Offset=<0 0 15>
    Sprite Size Mode=::NewEnumerator3
    Uniform Sprite Size=200
    UsePositionOffset=true
    Write Material Random=true
   )
   &SpriteFacingAndAlignment(Sprite Facing=<0 0 1>, WriteAlignment=false, WriteFacing=true)
  ]
  particleUpdateScript [
   &ParticleState()
   /Niagara/Modules/Decal/Decal_Attributes(
    Decal Fade:NiagaraFloat=&FloatFromCurve(FloatCurve=DataInterface() {propertyValues {Curve {keys [<at 0 0>,<at 0.066 1>,<at 0.6494 1>,<at 1 0>]}}})
    Decal Size:Vector3f=&VectorFromCurve(
     VectorCurve=DataInterface() {propertyValues {XCurve {keys [<at 0 1>,<at 1 1>]}, YCurve {keys [<at 0 0.5>,<at 0.025 1>]}, ZCurve {keys [<at 0 0.5>,<at 0.025 1>]}}}
     Scale Curve=<5 150 150>
    )
    Fade=1
    Size=1
   )
  ]
  rendererClasses [&NiagaraSpriteRendererProperties, &NiagaraDecalRendererProperties]
  renderers [
   {rendererClass=&NiagaraSpriteRendererProperties}
   {rendererClass=&NiagaraDecalRendererProperties, rendererIndex=1}
  ]
 }
 Rocks_Burst {
  emitterUpdateScript [
   &EmitterState(Life Cycle Mode=::NewEnumerator1, Loop Behavior=::NewEnumerator1, Loop Delay=1.438, Loop Duration=1, UseLoopDelay=true)
   &SpawnBurst_Instantaneous(
    Spawn Count=&RandomRangeInt(Minimum=5, Maximum=7)
    Use Spawn Probability=false
    UseLoopCountLimit=false
   )
   /SetVariables() {isSetParameters=true}
  ]
  particleSpawnScript [
   &InitializeParticle(
    Lifetime Max=0.7
    Lifetime Min=0.5
    Lifetime Mode=::NewEnumerator1
    Mesh Scale Mode=::NewEnumerator3
    Mesh Uniform Scale=&Multiply_Float(
     A=&Lerp_Float(
      A=0.0285
      B=0.06
      Alpha=&FloatFromCurve(
       FloatCurve=DataInterface() {propertyValues {Curve {keys [<at 0 0 RCIM_Constant>, {arriveTangent=7.6187, interpMode=RCIM_Constant, leaveTangent=7.6187, time=0.65, value=1}]}}}
       CurveIndex=&RandomRangeFloat(Minimum=0, Maximum=1)
      )
     )
     B=1.4
    )
    UsePositionOffset=false
    Write Material Random=true
   )
   &ShapeLocation(
    Handle Radius=0
    Large Radius=2
    MaskSpawnGroup=false
    Offset=<0 -7 47>
    Offset Mode=::NewEnumerator0
    Override Seed=false
    Shape Primitive=::NewEnumerator3
    UseOwnerScale=true
   )
   &InitialMeshOrientation(Use Rotation Vector=false)
   &AddRotationalVelocity(Rotation Rate=&Multiply_VectorByFloat(
     Vector=&RandomVector(Vector Scale=2)
     Float=&RandomRangeFloat(Minimum=0.5, Maximum=1)
    ))
   &DynamicMaterialParameters(
    Index 0 Param 2=1
    Module.Param0WriteEnabled=false
    Module.Param1WriteEnabled=true
    Module.Param2WriteEnabled=false
    Module.Param3WriteEnabled=false
   )
   &AddVelocity(
    Constrain To Radius=1000
    Module.Use Falloff Distance=false
    Origin Offset=<2 -5 40>
    Override Seed=false
    Velocity Mode=::NewEnumerator1
    Velocity Speed=&RandomRangeFloat(Minimum=200, Maximum=50)
   )
   &AddVelocity(
    Cone Angle=100
    Cone Axis=<0 0 1>
    Inner Cone Angle=30
    Module.UseConeFalloff=false
    Module.UseVelDistribution=false
    Override Seed=false
    Velocity Mode=::NewEnumerator2
    Velocity Speed=&RandomRangeFloat(Minimum=50, Maximum=150)
   ) {module=AddVelocity001}
   &ApplyInitialForces()
   /SetVariables(Particles.MeshIndex=&RandomRangeInt(Minimum=0, Maximum=2)) {isSetParameters=true}
  ]
  particleUpdateScript [
   &ParticleState()
   &UpdateMeshOrientation(Rotation Rate=2)
   &GravityForce(Gravity=&RandomRangeVector(Minimum=<0 0 -750>, Maximum=<0 0 -980>))
   &Collision(
    Age Colliding Particles=false
    KillOnCollision=false
    Module.Control Rest State Via Surface Normal=false
    Module.EnableMaxCollisionCount=false
    Module.Randomize Collision Normal Vector=false
    Restitution=0.2
   ) {enabled=false}
   &Collision(
    Age Colliding Particles=false
    Collision Enabled=&Boolean_AndOperation(
     Bool1=&IsPlatformSetActive(New Platform Set=DataInterface() {propertyValues {Platforms {qualityLevelMask=-8}}})
     Bool2=&Boolean_NotOperation(Bool=$User.bIsFrontEndPreview)
    )
    Control Roll On Collision=true
    Enable Rest State=false
    Friction=0.1
    KillOnCollision=false
    Module.EnableMaxCollisionCount=false
    Module.Randomize Collision Normal Vector=false
    Particle Radius Scale=5
    Radius Calculation Type=::NewEnumerator1
    Restitution=0.4
    ScaleRateOfRotation=true
   ) {enabled=false, module=Collision001}
   &Collision(
    Age Colliding Particles=false
    Analytical Collision Plane Space:ENiagaraCoordinateSpace=::Local
    CPU Collision Type=::NewEnumerator3
    Collision Enabled=&Boolean_OrOperation(
     Bool1=&IsPlatformSetActive(New Platform Set=DataInterface() {propertyValues {Platforms {qualityLevelMask=-28}}})
     Bool2=$User.bIsFrontEndPreview
    )
    Control Roll On Collision=true
    Enable Rest State=false
    Friction=0.1
    KillOnCollision=false
    Module.EnableMaxCollisionCount=false
    Module.Randomize Collision Normal Vector=false
    Particle Radius Scale=5
    Perform 2 Analytical Collision Queries=false
    Radius Calculation Type=::NewEnumerator1
    Restitution=0.8
    ScaleRateOfRotation=false
   ) {enabled=false, module=Collision002}
   &Drag(
    Drag=1.5
    Rotational Drag:NiagaraFloat=&FloatFromCurve(FloatCurve=DataInterface() {propertyValues {Curve {keys [<at 0 0 RCIM_Cubic>, {arriveTangent=0.1846, interpMode=RCIM_Cubic, leaveTangent=0.1846, tangentMode=RCTM_User, time=1, value=0.1}]}}})
    Use Linear Drag=true
    Use Rotational Drag=true
   )
   &SolveForcesAndVelocity(Clamp Velocity=false, Limit Acceleration=false)
   &ScaleMeshSize(Scale Factor=&VectorFromFloat(Value=&FloatFromCurve(FloatCurve=DataInterface() {propertyValues {Curve {keys [{arriveTangent=-0.0035, interpMode=RCIM_Cubic, leaveTangent=-0.0035, tangentMode=RCTM_User, time=0.6, value=1}, <at 1 0>]}}})))
  ]
  rendererClasses [&NiagaraMeshRendererProperties]
  renderers [{rendererClass=&NiagaraMeshRendererProperties, rendererIndex=0}]
 }
}
summary {
 systemName=NS_Emote_HornetBerm_PropSpawn
 emitters [
  {emitterName=Debris, rendererClasses [&NiagaraMeshRendererProperties, &NiagaraSpriteRendererProperties]}
  {emitterName=Dust, rendererClasses [&NiagaraSpriteRendererProperties]}
  {emitterName=Dust001, rendererClasses [&NiagaraSpriteRendererProperties]}
  {
   bEnabled=false
   emitterName=GroundCrack
   rendererClasses [&NiagaraSpriteRendererProperties, &NiagaraDecalRendererProperties]
  }
  {emitterName=Rocks_Burst, rendererClasses [&NiagaraMeshRendererProperties]}
  {emitterName=Dust002, rendererClasses [&NiagaraSpriteRendererProperties]}
 ]
 userVariables [
  User.bIsFrontEndPreview {type=NiagaraBool}
  User.NormalizedNotifyProgress {type=NiagaraFloat, defaultValue=0, description=""}
 ]
}
