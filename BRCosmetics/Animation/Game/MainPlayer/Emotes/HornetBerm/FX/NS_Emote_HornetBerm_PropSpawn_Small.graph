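# Niagara system for the small prop-spawn effect in the HornetBerm emote, built from AirglowLWE and SmokeLWE emitters, whose names suggest lightweight emitters.
# It exposes float user parameters IsSword and NormalizedNotifyProgress, both defaulting to zero.
# A system spawn assignment and SystemState accompany dynamic inputs for float-to-bool comparison, bool-selected float and Vector2D values, and random ranges.
# The file lists no renderers and omits how these inputs are wired.
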NiagaraSystem /BRCosmetics/Animation/Game/MainPlayer/Emotes/HornetBerm/FX/NS_Emote_HornetBerm_PropSpawn_Small
dependencies {
 usedDataInterfaces []
 usedDynamicInputs [
  &MakeCustomFloatFromBool
  &SetBoolByFloatComparison
  &RandomRangeFloat
  &MakeCustomVector2DFromBool
  &RandomRangeVector2D
 ]
 usedModules [
  &self:SystemSpawnScript.SystemScriptSource.SystemScriptGraph.NiagaraNodeAssignment_0.SetVariables_#1
  &SystemState
 ]
}
emitters {AirglowLWE {rendererClasses [], renderers []}, SmokeLWE {rendererClasses [], renderers []}}
summary {
 systemName=NS_Emote_HornetBerm_PropSpawn_Small
 emitters [{emitterName=AirglowLWE, rendererClasses []}, {emitterName=SmokeLWE, rendererClasses []}]
 userVariables [
  User.IsSword {type=NiagaraFloat, defaultValue=0, description=""}
  User.NormalizedNotifyProgress {type=NiagaraFloat, defaultValue=0, description=""}
 ]
}
